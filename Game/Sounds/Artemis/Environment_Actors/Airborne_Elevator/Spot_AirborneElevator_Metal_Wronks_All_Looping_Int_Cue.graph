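# Looping ambient cue for the airborne elevator that plays sporadic metal stress wronks with indoor ambient class and spherical attenuation.
# Each loop waits a random delay, then a quality switch picks the branch.
# The first two quality slots randomly choose knock, knock plus stretch, stretch plus swell, or silence, with modulated layers.
# The other two use a smaller, louder knock-only pool excluded from branch culling, and concurrency caps overlapping instances.
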SoundCue /Game/Sounds/Artemis/Environment_Actors/Airborne_Elevator/Spot_AirborneElevator_Metal_Wronks_All_Looping_Int_Cue
root=Looping_0
nodes [
 {id=Delay_0, children [QualityLevel_0], properties {delayMax=45, delayMin=6}}
 {id=Looping_0, children [Delay_0]}
 {id=Mixer_1, children [Modulator_3,Modulator_4], properties {inputVolume [1,1]}}
 {id=Mixer_2, children [Modulator_4,Modulator_2], properties {inputVolume [1,1]}}
 {id=Modulator_0, children [Random_2], properties {pitchMax=1.1, pitchMin=0.9, volumeMax=2.5, volumeMin=2}}
 {id=Modulator_2, children [Random_5]}
 {id=Modulator_3, children [Random_0], properties {pitchMin=0.9, volumeMax=0.8, volumeMin=0.7}}
 {id=Modulator_4, children [Random_3], properties {pitchMin=0.9, volumeMax=0.8, volumeMin=0.6}}
 {id=QualityLevel_0, children [Random_4,Random_4,Modulator_0,Modulator_0]}
 {
  id=Random_0
  children [WavePlayer_1, WavePlayer_7, WavePlayer_0, WavePlayer_4, WavePlayer_3, WavePlayer_2, WavePlayer_20, WavePlayer_21, WavePlayer_17]
  properties {weights [1,1,1,1,1,1,1,1,1]}
 }
 {
  id=Random_2
  children [WavePlayer_32, WavePlayer_35, WavePlayer_34, WavePlayer_33]
  properties {bShouldExcludeFromBranchCulling=true, weights [1,1,1,1]}
 }
 {id=Random_3, children [WavePlayer_22, WavePlayer_23, WavePlayer_24], properties {weights [1,1,1]}}
 {id=Random_4, children [Modulator_3,Mixer_1,Mixer_2,null], properties {bRandomizeWithoutReplacement=false, weights [1,1,1,1]}}
 {id=Random_5, children [WavePlayer_8, WavePlayer_9, WavePlayer_10, WavePlayer_30, WavePlayer_31], properties {weights [1,1,1,1,1]}}
 {id=WavePlayer_0, properties {soundWaveAssetPtr=/Game/Sounds/Apollo/Biomes/Generic/AMB_Metal_Stress_Wronk_Groan_Knock_03}}
 {id=WavePlayer_1, properties {soundWaveAssetPtr=/Game/Sounds/Apollo/Biomes/Generic/AMB_Metal_Stress_Wronk_Groan_Knock_01}}
 {
  id=WavePlayer_10
  properties {soundWaveAssetPtr=/Game/Sounds/Apollo/Biomes/UFO_Crash_Sites/AmbientOneshots/AMB_UFO_Crashsite_Huge_Metal_Swell_03}
 }
 {id=WavePlayer_17, properties {soundWaveAssetPtr=/Game/Sounds/Apollo/Biomes/Generic/AMB_Metal_Stress_Wronk_Groan_Knock_09}}
 {id=WavePlayer_2, properties {soundWaveAssetPtr=/Game/Sounds/Apollo/Biomes/Generic/AMB_Metal_Stress_Wronk_Groan_Knock_06}}
 {id=WavePlayer_20, properties {soundWaveAssetPtr=/Game/Sounds/Apollo/Biomes/Generic/AMB_Metal_Stress_Wronk_Groan_Knock_07}}
 {id=WavePlayer_21, properties {soundWaveAssetPtr=/Game/Sounds/Apollo/Biomes/Generic/AMB_Metal_Stress_Wronk_Groan_Knock_08}}
 {id=WavePlayer_22, properties {soundWaveAssetPtr=/Game/Sounds/Apollo/Biomes/Generic/AMB_Metal_Stress_Wronk_Groan_Stretch_02}}
 {id=WavePlayer_23, properties {soundWaveAssetPtr=/Game/Sounds/Apollo/Biomes/Generic/AMB_Metal_Stress_Wronk_Groan_Stretch_03}}
 {id=WavePlayer_24, properties {soundWaveAssetPtr=/Game/Sounds/Apollo/Biomes/Generic/AMB_Metal_Stress_Wronk_Groan_Stretch_04}}
 {id=WavePlayer_3, properties {soundWaveAssetPtr=/Game/Sounds/Apollo/Biomes/Generic/AMB_Metal_Stress_Wronk_Groan_Knock_05}}
 {
  id=WavePlayer_30
  properties {soundWaveAssetPtr=/Game/Sounds/Apollo/Biomes/UFO_Crash_Sites/AmbientOneshots/AMB_UFO_Crashsite_Huge_Metal_Swell_04}
 }
 {
  id=WavePlayer_31
  properties {soundWaveAssetPtr=/Game/Sounds/Apollo/Biomes/UFO_Crash_Sites/AmbientOneshots/AMB_UFO_Crashsite_Huge_Metal_Swell_05}
 }
 {id=WavePlayer_32, properties {soundWaveAssetPtr=&AMB_Metal_Stress_Wronk_Groan_Knock_01}}
 {id=WavePlayer_33, properties {soundWaveAssetPtr=/Game/Sounds/Apollo/Biomes/Generic/AMB_Metal_Stress_Wronk_Groan_Knock_04}}
 {id=WavePlayer_34, properties {soundWaveAssetPtr=&AMB_Metal_Stress_Wronk_Groan_Knock_03}}
 {id=WavePlayer_35, properties {soundWaveAssetPtr=/Game/Sounds/Apollo/Biomes/Generic/AMB_Metal_Stress_Wronk_Groan_Knock_02}}
 {id=WavePlayer_4, properties {soundWaveAssetPtr=&AMB_Metal_Stress_Wronk_Groan_Knock_04}}
 {id=WavePlayer_7, properties {soundWaveAssetPtr=&AMB_Metal_Stress_Wronk_Groan_Knock_02}}
 {
  id=WavePlayer_8
  properties {soundWaveAssetPtr=/Game/Sounds/Apollo/Biomes/UFO_Crash_Sites/AmbientOneshots/AMB_UFO_Crashsite_Huge_Metal_Swell_01}
 }
 {
  id=WavePlayer_9
  properties {soundWaveAssetPtr=/Game/Sounds/Apollo/Biomes/UFO_Crash_Sites/AmbientOneshots/AMB_UFO_Crashsite_Huge_Metal_Swell_02}
 }
]
properties {
 attenuationSettings=/Game/Sounds/Attenuation/Fort_Metal_Stress_OneShot_Sphere_No_Occ_Att
 bOverrideConcurrency=true
 maxDistance=4000
 soundClassObject=&Fort_Ambient_Indoors_Athena
 volumeMultiplier=0.331
 concurrencyOverrides {maxCount=2, voiceStealReleaseTime=3}
 preEffectBusSends [{}]
}
